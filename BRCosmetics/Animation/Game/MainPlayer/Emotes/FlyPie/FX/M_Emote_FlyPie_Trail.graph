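# Additive, two-sided Niagara material for the smoky trail of the FlyPie emote, usable on ribbons, sprites and meshes.
# Panned UVs feed a channel-selected trail texture with particle-driven erosion, masked along its length and faded by view angle and depth.
# A biased normal map lit against sun vectors shades time-of-day tinted color and HDR emissive.
# A WPO subgraph layers sine wiggles with manual particle-driven tip and base offsets.
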Material /BRCosmetics/Animation/Game/MainPlayer/Emotes/FlyPie/FX/M_Emote_FlyPie_Trail
graph {
 properties {
  bTangentSpaceNormal=false
  bUsedWithNiagaraMeshParticles=true
  bUsedWithNiagaraRibbons=true
  bUsedWithNiagaraSprites=true
  blendMode=BLEND_Additive
  pixelDepthOffsetMode=PDOM_Legacy
  translucencyLightingMode=TLM_VolumetricPerVertexDirectional
  twoSided=true
  usedShadingModels=Default Lit
 }
 nodes [
  n1 = DynamicParameter() {parameterIndex=2, defaultValue=<rgba 0 0 1 1>, paramNames [Mask Progress, WPOZManualTip, WPOZManualBase, ""]}
  n2 = ScalarParameter() {defaultValue=500, group=WPO, parameterName=WPO_Z}
  n3 = ScalarParameter() {defaultValue=-300, group=WPO, parameterName=WPO_ZExtra}
  n4 = ScalarParameter() {defaultValue=-0.2, group=WPO, parameterName=WPO_WiggleMask}
  n5 = ScalarParameter() {defaultValue=3, group=WPO, parameterName=WPO_Y}
  n6 = NamedRerouteUsage()
  Composite!(WPOZManualTip=@n1.WPOZManualTip, WPOZ=@n2, WPOZExtra=@n3, WiggleMask=@n4, WPOY=@n5, TrailAlpha=@n6, WPOZManualBase=@n1.WPOZManualTip) {subgraphName=WPO Graph}
  n7 = DynamicParameter() {parameterIndex=2, defaultValue=<rgba 0 0 1 1>, paramNames [Mask Progress, WPOZManualTip, WPOZManualBase, ""]}
  n8 = Constant()
  n9 = TextureSample(UVs=Add(A=TextureCoordinate() {uTiling=0.5}, B=AppendVector(A=@n7.Mask Progress, B=@n8))) {samplerSource=SSM_Clamp_WorldGroupSettings, texture=/Game/Effects/Fort_Effects/Gameplay/Tutorial_Target/shader_mask_round_square_blurry}
  n10 = Multiply(A=NamedRerouteUsage(), B=@n9.R)
  n11 = VertexColor()
  n12 = ScalarParameter() {defaultValue=0.7, group=Alpha, parameterName=EndOpacity}
  n13 = LinearInterpolate(Alpha=Power(Base=@n11.G, Exp=@n12) {constExponent=0.1}) {constA=1, constB=0}
  n14 = LinearInterpolate(Alpha=Power(Base=@n11.B, Exp=@n12) {constExponent=0.5}) {constA=1, constB=0}
  Multiply(A=@n10, B=Saturate(None=Multiply(A=@n13, B=@n14)))
  n15 = DynamicParameter() {parameterIndex=1, defaultValue=<rgba 1 0 1 1>, paramNames [Erode, SphereMaskRadius, X_Offset, Y_Offset]}
  n16 = Add(
   A=AppendVector(A=ScalarParameter() {group=Ribbon_UVs, parameterName=UV_X_Offset}, B=ScalarParameter() {group=Ribbon_UVs, parameterName=UV_Y_Offset})
   B=Add(A=TextureCoordinate() {coordinateIndex=1, uTiling=2}, B=AppendVector(A=@n15.X_Offset, B=@n15.X_Offset))
  )
  n17 = VectorParameter() {
   desc="R = U, G=V, B=PanningSpeed"
   group=Ribbon_UVs
   parameterName=Trail _UVs
   channelNames {a="", b=PanningSpeed, g=V, r=U}
   defaultValue=<rgba 1 1 0.6 0>
  }
  n18 = AppendVector(A=@n17.U, B=@n17.U)
  n19 = DynamicParameter() {defaultValue=<rgba 1 0 1 1>, paramNames [Erode, SphereMaskRadius, PanSpeed, Emissive]}
  n20 = Time()
  n21 = Constant()
  NamedRerouteDeclaration(None=Panner(Coordinate=Multiply(A=@n16, B=@n18), Time=Multiply(A=@n19.PanSpeed, B=@n20), Speed=AppendVector(A=@n17.PanningSpeed, B=@n21))) {name=Trail Texture UVs, nodeColor=<rgba 0.2188 0 1 1>}
  n22 = TextureSampleParameter2D(UVs=NamedRerouteUsage()) {
   group=Trail_Normal
   parameterName=TrailNormal
   samplerType=SAMPLERTYPE_Normal
   texture=/Game/Effects/Fort_Effects/Textures/Smoke/T_Smoke_Trail_Soft_N
  }
  n23 = VectorParameter() {group=Trail_Normal, parameterName=NormalBias Direction, sortPriority=33, defaultValue=<rgba 0.5 0.5 0 0>}
  n24 = ComponentMask(None=@n23.RGB) {b=true, g=true, r=true}
  n25 = ScalarParameter() {desc=Toggle the Bias direction from Tangent to World Space, group=Trail_Normal, parameterName=NormalBias WorldSpace, sortPriority=35}
  n26 = Add(
   A=@n22.RGB
   B=LinearInterpolate(A=@n24, B=Transform(None=@n24) {transformSourceType=TRANSFORMSOURCE_World, transformType=TRANSFORM_Tangent}, Alpha=@n25)
  )
  n27 = VectorParameter() {group=Trail_Normal, parameterName=NormalBias Multiplier, sortPriority=34, defaultValue=<rgba 2 2 1 0>}
  n28 = ComponentMask(None=@n27.RGB) {b=true, g=true, r=true}
  NamedRerouteDeclaration(None=Normalize(VectorInput=Multiply(A=@n26, B=@n28))) {name=Biased Normal, nodeColor=<rgba 0.5234 0 1 1>}
  n29 = TextureSampleParameter2D(UVs=NamedRerouteUsage()) {group=Trail_Shape, parameterName=TrailTexture, texture=/Game/Effects/Fort_Effects/Textures/Smoke/T_Smoke_Trail_Soft_Chromatic}
  n30 = VectorParameter() {group=Trail_Shape, parameterName=TrailTexture_ChannelSelection, sortPriority=33, defaultValue=<rgba 0 0 1 0>}
  n31 = DotProduct(A=AppendVector(A=@n29.RGB, B=@n29.RGB), B=AppendVector(A=@n30.RGB, B=@n30.RGB))
  n32 = LinearInterpolate(
   A=ScalarParameter() {defaultValue=-3, group=Trail_Shape, parameterName=Erosion Min, sortPriority=34}
   B=ScalarParameter() {defaultValue=1, group=Trail_Shape, parameterName=Erosion Max, sortPriority=35}
   Alpha=Add(A=DynamicParameter() {defaultValue=<rgba 1 0 1 1>, paramNames [Erode, SphereMaskRadius, PanSpeed, Emissive]}.Erode, B=@n31)
  ) {constA=-3}
  NamedRerouteDeclaration(None=Clamp(None=Multiply(A=@n32, B=@n31))) {name=Trail Alpha, nodeColor=<rgba 0 0.4609 1 1>}
  NamedRerouteDeclaration(None=@n22.RGB) {name=Trail Normal, nodeColor=<rgba 0.7578 0 1 1>}
  n33 = ScalarParameter() {group=Alpha, parameterName=AngleFade}
  NamedRerouteDeclaration(None=Multiply(
    A=@n10
    B=MaterialFunctionCall("FadeSensitivity (optional)"=@n33) {materialFunction=&FadeBasedOnViewAngle}.Result
   )) {name=Final Final Alpha, nodeColor=<rgba 0 1 0.3906 1>}
  n34 = ComponentMask(None=TextureCoordinate()) {r=true}
  n35 = Constant() {r=0.5}
  n36 = Multiply(A=ParticleColor().A, B=NamedRerouteUsage())
  n37 = ScalarParameter() {defaultValue=150, group=Alpha, parameterName=Fade Length}
  n38 = ScalarParameter() {defaultValue=25, group=Alpha, parameterName=Fade Offset}
  n39 = DepthFade(
   Opacity=Multiply(
    A=@n36
    B=MaterialFunctionCall(Fade Length=@n37, Fade Offset=@n38) {materialFunction=&CameraDepthFade}.Result
   )
   FadeDistance=ScalarParameter() {defaultValue=40, group=Alpha, parameterName=DepthFadeDist}
  ) {fadeDistanceDefault=40}
  NamedRerouteDeclaration(None=Multiply(A=SphereMask(A=@n34, B=@n35, Radius=@n35) {attenuationRadius=0.5, hardnessPercent=90}, B=ShadingPathSwitch(Default=@n36, Deferred=@n39))) {name=Final Alpha, nodeColor=<rgba 0.9453 1 0 1>}
  n40 = VectorParameter() {group=Emissive, parameterName=Color Morning Multiplier, sortPriority=35, defaultValue=<rgba 0.9896 0.268 0.4929 0>}
  n41 = VectorParameter() {group=Emissive, parameterName=Color Evening Multiplier, sortPriority=37, defaultValue=<rgba 0.599 0.2901 0.3358 0>}
  n42 = VectorParameter() {group=Emissive, parameterName=Color Night Multiplier, sortPriority=38, defaultValue=<rgba 0.3498 0.4864 0.6458 0>}
  n43 = ParticleColor()
  n44 = Multiply(
   A=VectorParameter() {group=Base Color, parameterName=Color, defaultValue=<rgba 0.0896 0.1113 0.2 0>}.RGB
   B=Desaturation(None=@n43.RGB, Fraction=Constant() {r=0.5})
  )
  n45 = NamedRerouteUsage()
  n46 = Multiply(
   A=MaterialFunctionCall(
    Color_Morning=@n40.RGB
    Color_Day=VectorParameter() {group=Emissive, parameterName=Color Day Multiplier, sortPriority=36, defaultValue=<rgba 1 1 1 0>}.RGB
    Color_Evening=@n41.RGB
    Color_Night=@n42.RGB
   ) {materialFunction=&Color_Time_of_Day}.Result
   B=NamedRerouteDeclaration(None=LinearInterpolate(A=@n44, B=@n43.RGB, Alpha=@n45)) {name=Final_Base Color, nodeColor=<rgba 1 0.7266 0 1>}
  )
  n47 = NamedRerouteUsage()
  n48 = CollectionParameter() {
   collection=&FortniteMaterialParameters
   parameterName=SunlightDirectionalVector
  }
  n49 = LinearInterpolate(
   A=Multiply(A=@n46, B=ScalarParameter() {defaultValue=0.7, group=Alpha, parameterName=Shadow Multiplier}) {constB=0.74}
   B=@n46
   Alpha=DotProduct(A=Transform(None=@n47), B=ComponentMask(None=@n48) {b=true, g=true, r=true})
  )
  n50 = NamedRerouteUsage()
  NamedRerouteDeclaration(None=StaticSwitchParameter(True=Add(A=@n49, B=@n50), False=@n50) {group=Emissive, parameterName=IsShaderUnlit, sortPriority=34}) {name=Final_Emissive, nodeColor=<rgba 1 0.0703 0 1>}
  n51 = Transform(None=Multiply(A=NamedRerouteUsage(), B=Constant3Vector() {constant=<rgba -1 -1 1 0>}))
  n52 = CollectionParameter() {
   collection=&FortniteMaterialParameters
   parameterName=SunAndMoonModelDirectionalVector
  }
  n53 = ComponentMask(None=@n52) {b=true, g=true, r=true}
  n54 = Saturate(None=Abs(None=DotProduct(A=@n51, B=@n53)))
  n55 = Constant() {r=0.5}
  n56 = Multiply(A=NamedRerouteUsage(), B=Multiply(A=ShadingPathSwitch(Default=@n54, Forward=@n55, Mobile=@n55), B=NamedRerouteUsage()))
  n57 = MaterialFunctionCall(
   LDR Input=Saturate(None=@n56)
   HDR Tint=@n56
   HDR Intensity Multiplier=Multiply(
    A=ScalarParameter() {defaultValue=25, group=Emissive, parameterName=HDR_Intensity}
    B=DynamicParameter() {defaultValue=<rgba 1 0 1 1>, paramNames [Erode, SphereMaskRadius, PanSpeed, Emissive]}.Emissive
   )
  ) {materialFunction=&DeriveHDRfromLDR}
  n58 = TextureSampleParameter2D() {
   group=Emissive
   parameterName=DetailTexture
   sortPriority=33
   texture=/Game/Global/Textures/Default/Tileable_Patterns/T_Blowing_Smoke_Pattern2_Tile_DISP
  }
  n59 = DynamicParameter() {parameterIndex=1, defaultValue=<rgba 1 0 1 1>, paramNames [Erode, SphereMaskRadius, X_Offset, Y_Offset]}
  n60 = Power(Base=@n58.G, Exp=@n59.Erode)
  n61 = DynamicParameter() {parameterIndex=1, defaultValue=<rgba 1 0 1 1>, paramNames [Erode, SphereMaskRadius, X_Offset, Y_Offset]}
  NamedRerouteDeclaration(None=LinearInterpolate(A=@n57.RemappedRange, B=Multiply(A=@n57.RemappedRange, B=@n60), Alpha=@n61.SphereMaskRadius)) {name=HDR Emissive, nodeColor=<rgba 1 0.8438 0 1>}
  n62 = ComponentMask(None=TextureCoordinate()) {r=true}
  n63 = Time()
  n64 = Sine(None=Add(A=Multiply(A=@n63) {constB=0.3}, B=@n62))
  n65 = Constant3Vector() {constant=<rgba 0 0 1 0>}
  n66 = OneMinus(None=@n62)
  MaterialExpressionReroute_18 = Reroute(None=@n4)
  n67 = Add(A=Divide(A=@n66), B=@MaterialExpressionReroute_18.WiggleMask)
  MaterialExpressionReroute_13 = Reroute(None=@n2)
  n68 = Multiply(A=Multiply(A=Multiply(A=@n64, B=@n65), B=Saturate(None=@n67)), B=@MaterialExpressionReroute_13.WPOZ) {constB=80}
  MaterialExpressionReroute_12 = Reroute(None=@n1.WPOZManualTip)
  n69 = Multiply(A=@n62, B=@MaterialExpressionReroute_12.WPOZManualTip)
  n70 = Add(A=Multiply(A=Add(A=@n63) {constB=-0.25}) {constB=-0.45}, B=@n62)
  n71 = Add(A=Multiply(A=Add(A=@n63) {constB=0.1}) {constB=-0.15}, B=@n62)
  n72 = Multiply(A=Sine(None=Multiply(A=@n70) {constB=0.5}), B=Sine(None=Multiply(A=@n71) {constB=0.75}))
  n73 = Multiply(A=Add(A=@n63) {constB=0.41}) {constB=-0.3}
  n74 = Sine(None=Multiply(A=Add(A=@n73, B=@n62)))
  n75 = Multiply(A=Clamp(None=Multiply(A=@n72, B=@n74)) {minDefault=-1}, B=@n65)
  MaterialExpressionReroute_15 = Reroute(None=@n3)
  n76 = Multiply(A=Add(A=@n62) {constB=-0.35}, B=Add(A=@n66) {constB=-0.15})
  n77 = Add(
   A=Add(A=Multiply(A=@n68, B=@n62), B=Multiply(A=@n69, B=@n65))
   B=Multiply(A=Multiply(A=@n75, B=@MaterialExpressionReroute_15.WPOZExtra), B=Saturate(None=@n76))
  )
  MaterialExpressionReroute_20 = Reroute(None=@n6)
  MaterialExpressionReroute_19 = Reroute(None=@n5)
  n78 = Multiply(
   A=@MaterialExpressionReroute_19.WPOY
   B=Multiply(A=@MaterialExpressionReroute_20.TrailAlpha, B=Constant3Vector() {constant=<rgba 0 1 0 0>}.G)
  )
  n79 = Multiply(A=Saturate(None=Add(A=@n66) {constB=-0.2}), B=@n65)
  MaterialExpressionReroute_21 = Reroute(None=@n1.WPOZManualBase)
  n80 = Multiply(A=Add(A=Add(A=@n77, B=@n78), B=Multiply(A=@n79, B=@MaterialExpressionReroute_21.WPOZManualBase)), B=@n62)
  MaterialExpressionReroute_16 = Reroute(None=@n80)
  NamedRerouteDeclaration(None=@MaterialExpressionReroute_16.Output) {name=WPOFinal, nodeColor=<rgba 0.6875 0 1 1>}
  NamedRerouteUsage()
  NamedRerouteUsage()
  NamedRerouteUsage()
  NamedRerouteUsage()
  NamedRerouteUsage()
  PinBase() {reroutePins [
    WPOZManualTip {expression=&self:MaterialExpressionReroute_12}
    WPOZ {expression=&self:MaterialExpressionReroute_13}
    WPOZExtra {expression=&self:MaterialExpressionReroute_15}
    WiggleMask {expression=&self:MaterialExpressionReroute_18}
    WPOY {expression=&self:MaterialExpressionReroute_19}
    TrailAlpha {expression=&self:MaterialExpressionReroute_20}
    WPOZManualBase {expression=&self:MaterialExpressionReroute_21}
   ]}
  PinBase!(Output=@n80) {reroutePins [Output {expression=&self:MaterialExpressionReroute_16}]}
 ]
}
