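# Niagara system for the FlyPie emote: steam with a warm glow and rising sparkles that track a sampled skeletal mesh.
# SteamLocal re-samples the mesh each update to stay attached, while SteamWorld samples only at spawn and tints by time of day.
# CrownGlow adds sprite and light; Reload_Smoke is a mesh plume shaped by dynamic material parameters from User Distance, ZRaise and Shorten.
# Debug, ribbon smoke and empty emitters are disabled.
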NiagaraSystem /BRCosmetics/Animation/Game/MainPlayer/Emotes/FlyPie/FX/NS_Emote_FlyPie_Steam
dependencies {
 usedDataInterfaces [&NiagaraDataInterfaceCurve, &NiagaraDataInterfaceColorCurve, &NiagaraDataInterfaceSkeletalMesh]
 usedDynamicInputs []
 usedModules [
  &SystemState
  &self:SystemSpawnScript.SystemScriptSource.SystemScriptGraph.NiagaraNodeAssignment_0.SetVariables_#1
 ]
}
emitters {
 CrownGlow {
  emitterUpdateScript [
   &EmitterState(Life Cycle Mode=::NewEnumerator1, Loop Behavior=::NewEnumerator1, Loop Duration=25, UseLoopDelay=false)
   &SpawnBurst_Instantaneous(Use Spawn Probability=false, UseLoopCountLimit=false)
  ]
  particleSpawnScript [&InitializeParticle(
    Color=<rgba 5 2 0 0.5>
    Lifetime=0.3
    Sprite Size Mode=::NewEnumerator3
    Uniform Sprite Size=75
    UsePositionOffset=false
    Write Material Random=true
   )]
  particleUpdateScript [
   &ParticleState(Kill Particles When Lifetime Has Elapsed=false)
   &SkeletalMeshLocation(
    Bone Sampling Mode=::NewEnumerator4
    MaskSpawnGroup=false
    Orientation Sampling=::NewEnumerator1
    Sampled Position Offset=<0 0 13>
    Skeletal Mesh=$System.Skeletal Mesh
   ) {module=SkeletalMeshLocation001}
   &CameraOffset(Camera Offset Amount=5)
   /SetVariables(
    Particles.GlowSize=&Multiply_Vector2D_ByFloat(
     Float=&FloatFromCurve(
      FloatCurve=DataInterface() {propertyValues {Curve {keys [
          {arriveTangent=6.7437, interpMode=RCIM_Cubic, leaveTangent=6.7437, tangentMode=RCTM_User, time=0, value=0}
          {arriveTangent=1.0225, interpMode=RCIM_Cubic, leaveTangent=1.0225, tangentMode=RCTM_User, time=2, value=4}
          <at 2 4>
         ]}}}
      CurveIndex=$Emitter.Age
      Scale Curve=50
     )
     Vector2D=<1 1>
    )
    Particles.SpriteSize=&Multiply_Vector2D_ByFloat(
     Float=&FloatFromCurve(
      FloatCurve=DataInterface() {propertyValues {Curve {keys [
          {arriveTangent=34.6115, interpMode=RCIM_Cubic, leaveTangent=34.6115, tangentMode=RCTM_User, time=0, value=0}
          {arriveTangent=4.6475, interpMode=RCIM_Cubic, leaveTangent=4.6475, tangentMode=RCTM_User, time=0.2, value=3.75}
          <at 0.7 4.625>
          {arriveTangent=-0.0405, interpMode=RCIM_Cubic, leaveTangent=-0.0405, tangentMode=RCTM_User, time=1, value=4.375}
         ]}}}
      Scale Curve=30
     )
     Vector2D=<0.65 0.65>
    )
   ) {isSetParameters=true}
   &ScaleColor(
    Scale Alpha=&FloatFromCurve(
     FloatCurve=DataInterface() {propertyValues {Curve {keys [
         {arriveTangent=-0.2933, interpMode=RCIM_Cubic, leaveTangent=-0.2933, tangentMode=RCTM_User, time=0, value=0}
         {arriveTangent=-2.8313, interpMode=RCIM_Cubic, leaveTangent=-2.8313, tangentMode=RCTM_User, time=0.1, value=0.95}
         {arriveTangent=-1.6622, interpMode=RCIM_Cubic, leaveTangent=-1.6622, tangentMode=RCTM_User, time=0.2, value=0.45}
         {arriveTangent=-0.0383, interpMode=RCIM_Cubic, leaveTangent=-0.0383, tangentMode=RCTM_User, time=1, value=0.2}
        ]}}}
     Scale Curve=0.4
    )
    ScaleA=true
    ScaleRGB=false
   ) {module=ScaleColor001}
  ]
  rendererClasses [&NiagaraSpriteRendererProperties, &NiagaraLightRendererProperties]
  renderers [
   {rendererClass=&NiagaraSpriteRendererProperties}
   {rendererClass=&NiagaraLightRendererProperties, rendererIndex=1}
  ]
 }
 Empty {bEnabled=false, rendererClasses [], renderers []}
 FN_Debug {
  bEnabled=false
  emitterUpdateScript [
   &EmitterState()
   &SpawnBurst_Instantaneous(Spawn Count=1, Use Spawn Probability=false)
  ]
  particleSpawnScript [/Niagara/Modules/Spawn/Initialization/InitializeParticle(
    Lifetime=5
    Module.Write Color=true
    Module.Write Lifetime=true
    Module.Write Mass=true
    Module.Write Material Random=true
    Module.Write Position=false
    Module.Write Scale=false
    Module.Write SpriteRotation=false
    Module.Write SpriteSize=true
    Position=&SimulationPosition()
    Sprite Size=<100 100>
   )]
  particleUpdateScript [
   &ParticleState()
   /SetVariables(Particles.Debug:NiagaraFloat=&RemapRange(Input Value To Remap=$User.ZRaise, Input Min=0, Input Max=1, Output Min=0, Output Max=1)) {isSetParameters=true}
   &DynamicMaterialParameters(
    Index 0 Param 1=$Particles.Debug
    Module.Param0WriteEnabled=true
    Module.Param1WriteEnabled=true
    Module.Param2WriteEnabled=true
    Module.Param3WriteEnabled=true
   )
  ]
  rendererClasses [&NiagaraSpriteRendererProperties]
  renderers [{rendererClass=&NiagaraSpriteRendererProperties}]
 }
 Reload_Smoke {
  emitterUpdateScript [
   &EmitterState(Life Cycle Mode=::NewEnumerator1, Loop Behavior=::NewEnumerator1, Loop Duration Mode=::NewEnumerator1, UseLoopDelay=false)
   &SpawnBurst_Instantaneous(Use Spawn Probability=false, UseLoopCountLimit=false)
   &SpawnRate(SpawnRate=2, Use Spawn Probability=false) {enabled=false}
   /SetVariables(
    Emitter.ZRaiseEmitter:NiagaraFloat=&RemapRange(Input Value To Remap=$User.ZRaise, Input Min=0, Input Max=1, Output Min=-20, Output Max=15)
    Emitter.float:NiagaraFloat=&RemapRange(Input Value To Remap=$User.Shorten, Input Min=0, Input Max=1, Output Min=10, Output Max=-110)
   ) {isSetParameters=true}
  ]
  particleSpawnScript [/Niagara/Modules/Spawn/Initialization/V2/InitializeParticle(
    Color=$User.BaseColorAlpha
    Mesh Scale Mode=::NewEnumerator3
    Ribbon Facing Mode=::NewEnumerator1
    Ribbon Facing Vector:Vector3f=<1 0 0>
    UsePositionOffset=false
    Write Material Random=true
   )]
  particleUpdateScript [
   &ParticleState(Kill Particles When Lifetime Has Elapsed=false)
   &ScaleMeshSize(Scale Factor=&MakeVector(
     X=1
     Y=&FloatFromCurve(
      FloatCurve=DataInterface() {propertyValues {Curve {keys [{tangentMode=RCTM_User, time=0, value=0.7}, {tangentMode=RCTM_User, time=1, value=1}]}}}
      Scale Curve=&RemapRange(Input Value To Remap=$User.Distance, Input Min=0, Input Max=1, Output Min=0.6, Output Max=1.1)
     )
     Z=1
    ))
   &SkeletalMeshLocation(
    Bone Sampling Mode=::NewEnumerator4
    MaskSpawnGroup=false
    Orientation Sampling=::NewEnumerator1
    Sampled Position Offset=<0 0 20>
    Skeletal Mesh=$System.Skeletal Mesh
   )
   &ScaleColor(Scale Alpha=0.2, ScaleA=true, ScaleRGB=true)
   &DynamicMaterialParameters(
    Index 0 Param 1=1
    Index 0 Param 3=-1
    Index 0 Param 4=&MakeCustomFloatFromBool(
     Bool=$User.UseEmissiveSmoke
     True Float=&FloatFromCurve(FloatCurve=DataInterface() {propertyValues {Curve {keys [{arriveTangent=-2.94, interpMode=RCIM_Cubic, time=0, value=1}, <at 0.75 0 RCIM_Cubic>]}}})
     False Float=0
    )
    Index 1 Param 1=&FloatFromCurve(FloatCurve=DataInterface() {propertyValues {Curve {keys [<at 0 1 RCIM_Cubic>, {arriveTangent=45.3807, interpMode=RCIM_Cubic, leaveTangent=45.3807, tangentMode=RCTM_User, time=0.5, value=10}]}}})
    Index 1 Param 2=&FloatFromCurve(FloatCurve=DataInterface() {propertyValues {Curve {keys [{arriveTangent=-2.2405, interpMode=RCIM_Cubic, leaveTangent=-2.2405, tangentMode=RCTM_User, time=0, value=1}, <at 1 0 RCIM_Cubic>]}}})
    Index 2 Param 1=&ClampFloat(
     Float=&Add_Float(
      A=&FloatFromCurve(FloatCurve=DataInterface() {propertyValues {Curve {keys [<at 0 1>,<at 0.3 0.1>]}}}, CurveIndex=$Emitter.Age)
      B=&Add_Float(
       A=&ClampFloat(Float=&Waveform("Waveform [1]"=::NewEnumerator0, "[1] Frequency"=0.2, "[1] Amplitude Min/Max"=<0 0.3>), Min=-1)
       B=&RemapRange(Input Value To Remap=$User.Distance, Input Min=0, Input Max=1, Output Min=0.05, Output Max=0.1)
      )
     )
     Min=0.35
    )
    Index 2 Param 2=$Emitter.ZRaiseEmitter
    Index 2 Param 3=&ClampFloat(
     Float=$Emitter.float
     Min=-500
     Max=&FloatFromCurve(FloatCurve=DataInterface() {propertyValues {Curve {keys [<at 0.6 0>,<at 1 100>]}}}, CurveIndex=$Emitter.Age)
    )
    Module.Param0WriteEnabled=true
    Module.Param0WriteEnabled1=true
    Module.Param0WriteEnabled2=true
    Module.Param1WriteEnabled=false
    Module.Param1WriteEnabled1=true
    Module.Param1WriteEnabled2=true
    Module.Param2WriteEnabled=true
    Module.Param2WriteEnabled1=false
    Module.Param2WriteEnabled2=true
    Module.Param3WriteEnabled=true
    Module.Param3WriteEnabled1=false
    Module.Param3WriteEnabled2=true
    Write Parameter Index 1=true
    Write Parameter Index 2=true
   )
   &CurlNoiseForce(
    Mask Curl Noise=true
    Noise Frequency=10
    Noise Quality / Cost=::NewEnumerator0
    Noise Strength=625
    Pan Noise=true
    Pan Noise Field=<-10 0 0>
   ) {enabled=false}
   &InheritSourceMovement(
    Applied Position Scale=&Multiply_VectorByFloat(
     Vector=<1 1 1>
     Float=&FloatFromCurve(FloatCurve=DataInterface() {propertyValues {Curve {keys [<at 0 1>,<at 1 0.6>]}}})
    )
    ApplyForce=false
    ApplyPosition=true
    ApplyPositionFalloff=false
   ) {enabled=false}
   &AccelerationForce(Acceleration=&MakeVector(
     X=0
     Y=0
     Z=&FloatFromCurve(FloatCurve=DataInterface() {propertyValues {Curve {keys [<at 0 -50>,<at 0.4 -300>,<at 0.5 400>]}}})
    )) {enabled=false}
   &SolveForcesAndVelocity(Clamp Velocity=false, Limit Acceleration=false)
  ]
  rendererClasses [&NiagaraMeshRendererProperties]
  renderers [{rendererClass=&NiagaraMeshRendererProperties, rendererIndex=0}]
 }
 Reload_Smoke001 {
  bEnabled=false
  emitterUpdateScript [
   &EmitterState()
   &SpawnRate(SpawnRate=64, Use Spawn Probability=false)
  ]
  particleSpawnScript [
   &InitializeParticle(
    Color=$User.BaseColorAlpha
    Lifetime=&FloatFromCurve(
     FloatCurve=DataInterface() {propertyValues {Curve {keys [<at 0 1 RCIM_Cubic>, {arriveTangent=-1.028, interpMode=RCIM_Cubic, leaveTangent=-1.028, tangentMode=RCTM_User, time=1, value=0.5}]}}}
     CurveIndex=$Emitter.NormalizedLoopAge
     Scale Curve=$User.RibbonLifetime
    )
    Ribbon Width=&Multiply_Float(A=9, B=$User.RibbonScale)
    Ribbon Width Mode=::NewEnumerator1
    UsePositionOffset=false
    Write Material Random=true
   )
   &SkeletalMeshLocation(
    Bone Sampling Mode=::NewEnumerator4
    MaskSpawnGroup=false
    Orientation Sampling=::NewEnumerator1
    Sampled Position Offset=<0 0 20>
    Skeletal Mesh=$System.Skeletal Mesh
   )
   &AddVelocity(Velocity=<0 -50 10>)
  ]
  particleUpdateScript [
   &ParticleState()
   &ScaleColor(
    Scale Alpha=&Multiply_Float(
     A=&FloatFromCurve(
      FloatCurve=DataInterface() {propertyValues {Curve {keys [<at 0 0>, {arriveTangent=-1.7714, interpMode=RCIM_Cubic, leaveTangent=-1.7714, tangentMode=RCTM_User, time=0.1, value=1}, <at 1 0>]}}}
      Scale Curve=&FNTimeOfDayScalar(Night Scale=1, Morning Scale=0.3, Evening Scale=0.5)
     )
     B=&FloatFromCurve(FloatCurve=DataInterface() {propertyValues {Curve {keys [<at 0 1>,<at 0.75 1>,<at 1 0>]}}}, CurveIndex=$Emitter.NormalizedLoopAge)
    )
    Scale RGB=&MakeVectorFromLinearColorRGB(Color=&MakeCustomLinearColorFromBool(
      Bool=$User.UseEmissiveSmoke
      True Color=&LerpLinearColors(
       StartColor=$User.EmissiveColor
       EndColor=&MultiplyLinearColor(Linear Color=$User.EmissiveColor, Scale Factor=&MakeVector4_FromFloat(Float=0.25))
       LerpFactor=&FloatFromCurve(FloatCurve=DataInterface() {propertyValues {Curve {keys [<at 0 0 RCIM_Cubic>, {arriveTangent=2.2656, interpMode=RCIM_Cubic, leaveTangent=2.2656, tangentMode=RCTM_User, time=1, value=1}]}}})
      )
      False Color=<rgba 1 1 1 1>
     ))
    ScaleA=true
    ScaleRGB=true
   )
   &DynamicMaterialParameters(
    Index 0 Param 1=&FloatFromCurve(FloatCurve=DataInterface() {propertyValues {Curve {keys [
         <at 0 1 RCIM_Cubic>
         {arriveTangent=-1.7387, arriveTangentWeight=1.9926, interpMode=RCIM_Cubic, leaveTangent=-1.7387, tangentMode=RCTM_User, time=1, value=0.125}
        ]}}})
    Index 0 Param 3=-1
    Index 0 Param 4=&MakeCustomFloatFromBool(
     Bool=$User.UseEmissiveSmoke
     True Float=&FloatFromCurve(FloatCurve=DataInterface() {propertyValues {Curve {keys [{arriveTangent=-2.94, interpMode=RCIM_Cubic, time=0, value=1}, <at 0.75 0 RCIM_Cubic>]}}})
     False Float=0
    )
    Index 1 Param 1=&FloatFromCurve(FloatCurve=DataInterface() {propertyValues {Curve {keys [<at 0 1 RCIM_Cubic>, {arriveTangent=45.3807, interpMode=RCIM_Cubic, leaveTangent=45.3807, tangentMode=RCTM_User, time=0.5, value=10}]}}})
    Index 1 Param 2=&FloatFromCurve(FloatCurve=DataInterface() {propertyValues {Curve {keys [{arriveTangent=-2.2405, interpMode=RCIM_Cubic, leaveTangent=-2.2405, tangentMode=RCTM_User, time=0, value=1}, <at 1 0 RCIM_Cubic>]}}})
    Module.Param0WriteEnabled=true
    Module.Param0WriteEnabled1=true
    Module.Param1WriteEnabled=false
    Module.Param1WriteEnabled1=true
    Module.Param2WriteEnabled=true
    Module.Param2WriteEnabled1=false
    Module.Param3WriteEnabled=true
    Module.Param3WriteEnabled1=false
    Write Parameter Index 1=true
   )
   &ScaleRibbonWidth(Ribbon Width Scale=&FloatFromCurve(FloatCurve=DataInterface() {propertyValues {Curve {keys [{tangentMode=RCTM_User, time=0, value=1}, {tangentMode=RCTM_User, time=1, value=0.125}]}}}))
   &CurlNoiseForce(
    Mask Curl Noise=true
    Noise Frequency=10
    Noise Quality / Cost=::NewEnumerator0
    Noise Strength=50
    Pan Noise=true
    Pan Noise Field=<-10 0 0>
   )
   &InheritSourceMovement(
    Applied Position Scale=&Multiply_VectorByFloat(
     Vector=<1 1 1>
     Float=&FloatFromCurve(FloatCurve=DataInterface() {propertyValues {Curve {keys [<at 0 1>,<at 1 0.6>]}}})
    )
    ApplyForce=false
    ApplyPosition=true
    ApplyPositionFalloff=false
   ) {enabled=false}
   &SolveForcesAndVelocity(Clamp Velocity=false, Limit Acceleration=false)
  ]
  rendererClasses [&NiagaraRibbonRendererProperties]
  renderers [{rendererClass=&NiagaraRibbonRendererProperties}]
 }
 Sparkles {
  emitterUpdateScript [
   &EmitterState(Life Cycle Mode=::NewEnumerator1, Loop Behavior=::NewEnumerator1, Loop Duration Mode=::NewEnumerator1, UseLoopDelay=false)
   &SpawnRate(SpawnRate=5, Use Spawn Probability=false)
  ]
  particleSpawnScript [
   &InitializeParticle(
    Color=<rgba 5 2.4004 0.5265 1>
    Lifetime=&UniformRangedFloat(Minimum=0.8, Maximum=2)
    Sprite Rotation Angle=45
    Sprite Rotation Mode=::NewEnumerator0
    Sprite Size Mode=::NewEnumerator3
    Uniform Sprite Size=&UniformRangedFloat(Minimum=20, Maximum=8)
    UsePositionOffset=false
    Write Material Random=true
   ) {module=InitializeParticle001}
   &AddVelocity(Velocity=&UniformRangedVector(Minimum=<-40 -40 -40>, Maximum=<40 40 40>))
   &SkeletalMeshLocation(
    Bone Sampling Mode=::NewEnumerator4
    MaskSpawnGroup=false
    Sampled Position Offset=&AddVectorToPosition(Position=<-15 -15 0>, Vector=<15 15 10>)
    Skeletal Mesh=$System.Skeletal Mesh
   )
   &ShapeLocation(
    MaskSpawnGroup=false
    Offset=<0 0 10>
    Offset Mode=::NewEnumerator0
    Override Seed=false
    Sphere Radius=10
    Sphere Surface Distribution=0.9048
    UseOwnerScale=true
   )
   &ApplyInitialForces()
   &CameraOffset(Camera Offset Amount=20)
  ]
  particleUpdateScript [
   &ParticleState()
   &ScaleSpriteSize(Uniform Curve Sprite Scale=DataInterface() {propertyValues {Curve {keys [
        {interpMode=RCIM_Cubic, tangentMode=RCTM_User, time=0, value=0.5718}
        {
         arriveTangent=-3.6492
         arriveTangentWeight=0.1689
         interpMode=RCIM_Cubic
         leaveTangent=-3.6492
         leaveTangentWeight=0.9359
         tangentMode=RCTM_User
         time=0.0814
         value=1
        }
        {tangentMode=RCTM_User, time=1, value=0}
       ]}}}) {module=ScaleSpriteSize001}
   &GravityForce(Gravity=<0 0 100>)
   &Drag(Drag=5, Use Linear Drag=true, Use Rotational Drag=false)
   &SolveForcesAndVelocity(Clamp Velocity=false, Limit Acceleration=false)
   &ScaleColor(
    Scale Alpha=&FloatFromCurve(FloatCurve=DataInterface() {propertyValues {Curve {keys [<at -0.0092 0.0144>,<at 0.0967 0.9767>,<at 1.0011 0.002>]}}})
    ScaleA=true
    ScaleRGB=true
   )
  ]
  rendererClasses [&NiagaraSpriteRendererProperties]
  renderers [{rendererClass=&NiagaraSpriteRendererProperties}]
 }
 SteamLocal {
  emitterUpdateScript [
   &EmitterState(Life Cycle Mode=::NewEnumerator1, Loop Duration=1, UseLoopDelay=true)
   &SpawnRate(SpawnRate=5, Use Spawn Probability=false)
  ]
  particleSpawnScript [
   &InitializeParticle(
    Color=<rgba 0.6875 0.6875 0.6875 0.2>
    Lifetime Min=1.5
    Lifetime Mode=::NewEnumerator1
    Position Offset=<0 0 4>
    Sprite Rotation Angle=&RandomRangeFloat(Minimum=-0.05, Maximum=0.3)
    Sprite Rotation Mode=::NewEnumerator1
    Sprite Size Mode=::NewEnumerator1
    Uniform Sprite Size Max=60
    Uniform Sprite Size Min=80
    UsePositionOffset=true
    Write Material Random=true
   )
   &ShapeLocation(MaskSpawnGroup=false, Override Seed=false, Sphere Radius=10, UseOwnerScale=true)
   &AddVelocity(Velocity=&RandomRangeVector(Minimum=<-2 -2 6>, Maximum=<2 2 10>, Evaluation Type=::NewEnumerator1))
   &InheritVelocity(Module.Limit Speed=true, Velocity Limit=5, Velocity Scale=<0.25 0.25 0.25>)
  ]
  particleUpdateScript [
   &ParticleState()
   &SubUVAnimation(
    Sub UV Play Rate:NiagaraFloat=&RandomRangeFloat(Minimum=0.05, Maximum=0.1)
    SubUV Animation Mode:ENiagara_SubUVLookupMode=::NewEnumerator3
    SubUV Start Time Offset:NiagaraFloat=&RandomRangeFloat(Minimum=0, Maximum=63)
   )
   &DynamicMaterialParameters(
    Index 0 Param 1=&FloatFromCurve(FloatCurve=DataInterface() {propertyValues {Curve {keys [<at 0 -0.25>,<at 1 -1>]}}})
    Index 0 Param 4=&OneMinusFloat(Float=$Particles.NormalizedAge)
    Module.Param0WriteEnabled=true
    Module.Param1WriteEnabled=false
    Module.Param2WriteEnabled=false
    Module.Param3WriteEnabled=true
   )
   &ScaleSpriteSize(Scale Factor=&Vector2DFromFloat(Value=&FloatFromCurve(
      FloatCurve=DataInterface() {propertyValues {Curve {keys [
          {arriveTangent=1.8751, interpMode=RCIM_Cubic, leaveTangent=1.8751, tangentMode=RCTM_User, time=0, value=0.25}
          {arriveTangent=0.72, interpMode=RCIM_Cubic, leaveTangent=0.72, tangentMode=RCTM_User, time=1, value=1.25}
         ]}}}
      Scale Curve=1.5
     )))
   &Drag(Drag=0.5, Use Linear Drag=true, Use Rotational Drag=false)
   &SkeletalMeshLocation(
    Bone Sampling Mode=::NewEnumerator4
    MaskSpawnGroup=false
    Orientation Sampling=::NewEnumerator1
    Sampled Position Offset=<0 0 20>
    Skeletal Mesh=$System.Skeletal Mesh
   ) {module=SkeletalMeshLocation001}
   &InheritVelocity(Module.Limit Speed=true, Velocity Limit=0.3, Velocity Scale=&VectorFromFloat(Value=0.01)) {module=InheritVelocity001}
   &InheritSourceMovement(
    Applied Position Scale=&VectorFromFloat(Value=&FloatFromCurve(FloatCurve=DataInterface() {propertyValues {Curve {keys [<at 0 1>,<at 1 0>]}}}))
    ApplyForce=false
    ApplyPosition=true
    ApplyPositionFalloff=false
   )
   &SolveForcesAndVelocity(Clamp Velocity=false, Limit Acceleration=false)
   &Color(Color=&ColorFromCurve(
     ColorCurve=DataInterface() {propertyValues {
       AlphaCurve {keys [<at 0 1 RCIM_Cubic>]}
       BlueCurve {keys [<at 0 0 RCIM_Cubic>,<at 0.5 0>,<at 1 1 RCIM_Cubic>]}
       GreenCurve {keys [<at 0 0 RCIM_Cubic>,<at 0.5 1>,<at 1 0 RCIM_Cubic>]}
       RedCurve {keys [<at 0 1 RCIM_Cubic>,<at 0.5 0>,<at 1 0 RCIM_Cubic>]}
      }}
     CurveIndex=$Particles.DistanceFromCup
    )) {enabled=false}
  ]
  rendererClasses [&NiagaraSpriteRendererProperties]
  renderers [{rendererClass=&NiagaraSpriteRendererProperties}]
 }
 SteamWorld {
  emitterUpdateScript [
   &EmitterState(Life Cycle Mode=::NewEnumerator1, Loop Duration=1, UseLoopDelay=true)
   &SpawnRate(SpawnRate=5, Use Spawn Probability=false)
  ]
  particleSpawnScript [
   &InitializeParticle(
    Color=&FNTimeOfDayColor(
     Night Color=<rgba 0.4757 0.4691 0.6875 0.14>
     Morning Color=<rgba 0.6875 0.4986 0.4878 0.113>
     Day Color=<rgba 0.6875 0.6875 0.6875 0.2>
     Evening Color=<rgba 0.6875 0.6875 0.6875 0.1>
    )
    Lifetime Max=1
    Lifetime Min=0.7
    Lifetime Mode=::NewEnumerator1
    Position Offset=<0 0 4>
    Sprite Rotation Angle=&RandomRangeFloat(Minimum=-0.05, Maximum=0.3)
    Sprite Rotation Mode=::NewEnumerator1
    Sprite Size Mode=::NewEnumerator1
    Uniform Sprite Size Max=60
    Uniform Sprite Size Min=80
    UsePositionOffset=true
    Write Material Random=true
   )
   &ShapeLocation(MaskSpawnGroup=false, Override Seed=false, Sphere Radius=10, UseOwnerScale=true)
   &SkeletalMeshLocation(
    Bone Sampling Mode=::NewEnumerator4
    MaskSpawnGroup=false
    Orientation Sampling=::NewEnumerator1
    Sampled Position Offset=<0 0 20>
    Skeletal Mesh=$System.Skeletal Mesh
   ) {module=SkeletalMeshLocation001}
   &AddVelocity(Velocity=&RandomRangeVector(Minimum=<-2 -40 0>, Maximum=<2 2 0>, Evaluation Type=::NewEnumerator1))
   &InheritVelocity(Module.Limit Speed=true, Velocity Limit=5, Velocity Scale=<0.25 0.25 0.25>)
  ]
  particleUpdateScript [
   &ParticleState()
   &SubUVAnimation(
    Sub UV Play Rate:NiagaraFloat=&RandomRangeFloat(Minimum=0.05, Maximum=0.1)
    SubUV Animation Mode:ENiagara_SubUVLookupMode=::NewEnumerator3
    SubUV Start Time Offset:NiagaraFloat=&RandomRangeFloat(Minimum=0, Maximum=63)
   )
   &DynamicMaterialParameters(
    Index 0 Param 1=&FloatFromCurve(FloatCurve=DataInterface() {propertyValues {Curve {keys [<at 0 -0.25>,<at 1 -1>]}}})
    Index 0 Param 4=&OneMinusFloat(Float=$Particles.NormalizedAge)
    Module.Param0WriteEnabled=true
    Module.Param1WriteEnabled=false
    Module.Param2WriteEnabled=false
    Module.Param3WriteEnabled=true
   )
   &ScaleSpriteSize(Scale Factor=&Vector2DFromFloat(Value=&FloatFromCurve(
      FloatCurve=DataInterface() {propertyValues {Curve {keys [
          {arriveTangent=1.8751, interpMode=RCIM_Cubic, leaveTangent=1.8751, tangentMode=RCTM_User, time=0, value=0.25}
          {arriveTangent=0.72, interpMode=RCIM_Cubic, leaveTangent=0.72, tangentMode=RCTM_User, time=1, value=1.25}
         ]}}}
      Scale Curve=1.5
     )))
   &Drag(Drag=0.5, Use Linear Drag=true, Use Rotational Drag=false)
   &SolveForcesAndVelocity(Clamp Velocity=false, Limit Acceleration=false)
  ]
  rendererClasses [&NiagaraSpriteRendererProperties]
  renderers [{rendererClass=&NiagaraSpriteRendererProperties}]
 }
}
summary {
 systemName=NS_Emote_FlyPie_Steam
 emitters [
  {emitterName=CrownGlow, rendererClasses [&NiagaraSpriteRendererProperties, &NiagaraLightRendererProperties]}
  {emitterName=SteamWorld, rendererClasses [&NiagaraSpriteRendererProperties]}
  {emitterName=Reload_Smoke, rendererClasses [&NiagaraMeshRendererProperties]}
  {bEnabled=false, emitterName=FN_Debug, rendererClasses [&NiagaraSpriteRendererProperties]}
  {emitterName=Sparkles, rendererClasses [&NiagaraSpriteRendererProperties]}
  {emitterName=SteamLocal, rendererClasses [&NiagaraSpriteRendererProperties]}
  {bEnabled=false, emitterName=Reload_Smoke001, rendererClasses [&NiagaraRibbonRendererProperties]}
  {bEnabled=false, emitterName=Empty, rendererClasses []}
 ]
 userVariables [
  User.BaseColorAlpha {type=LinearColor, description="", defaultValue {struct=LinearColor, value=<rgba 1 1 1 0.25>}}
  User.bIsRocketLauncher {type=NiagaraBool}
  User.Distance {type=NiagaraFloat, defaultValue=0, description=""}
  User.EmissiveColor {type=LinearColor, description="", defaultValue {struct=LinearColor, value=<rgba 0.1354 0.0043 0 1>}}
  User.RibbonLifetime {type=NiagaraFloat, defaultValue=2, description=""}
  User.RibbonMaterial {
   type=MaterialInterface
   description=""
   defaultValue {
    struct=NiagaraExt_VariableValue_Object
    value {object=/Game/Effects/Niagara/MasterMaterials/Reload/M_Reload_SmokeTrail, objectClass=/Script/Engine.MaterialInterface}
   }
  }
  User.RibbonScale {type=NiagaraFloat, defaultValue=8, description=""}
  User.Shorten {type=NiagaraFloat, defaultValue=0, description=""}
  User.UseEmissiveSmoke {type=NiagaraBool}
  User.ZRaise {type=NiagaraFloat, defaultValue=0, description=""}
 ]
}
